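; Geode object marker beam Blueprint whose visibility and fades follow calendar events.
; After match ready, it watches Beam On and Start events, storing seconds remaining or elapsed in TotalSeconds and setting RepNotify state.
; OnRep_BeamOn, client-only apart from flagging BeamActive, times material fades to outlast the event by Post Launch Duration.
; OnRep_ParameterValue sets the countdown collection's DisableLightcone and, if the beam is inactive, shows it without fade-in for remaining post-launch time.
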(event EventBeginPlay
  ;[1 1 1] Turn on beam and delay the fade out until NightNight event starts
  (Class|FortAsyncActionAuthorityMatchReady|MatchReadyAsync
    (:OnReady
      (Utilities|FlowControl|Sequence
        (:then_0
          (bind (_eventname _timeuntilend) (MCPContext|WatchCalendarEvent (Variables|Default|GetBeam%20On%20Event))
            (:OnEventActive
              (Variables|Default|SetTotalSeconds (Math|Timespan|GetTotalSeconds _timeuntilend))
              (Variables|Default|SetBeamOn true))))
        (:then_1
          (bind (_eventname_2 _timeuntilend_2 _timesincebegin) (MCPContext|WatchCalendarEvent (Variables|Default|GetStartEvent))
            (:OnEventActive
              (Variables|Default|SetTotalSeconds (Math|Timespan|GetTotalSeconds _timesincebegin))
              (Variables|Default|SetParameterValue 1.0))))))))

(fn OnRep_BeamOn
  (bind _beam (Variables|Default|GetBeam))
  (bind _beam_2 (Variables|Default|GetBeam))
  ;[1 1 1] Sets up Beams Visuals for Duration of Calendar Event based on TotalSeconds Left; ensures that if less than Fade In Time Remains the Beams starts On with no Fade In
  (Variables|Default|SetBeamActive true)
  (if (Networking|IsServer)
    _
    (elif (Variables|Default|GetBeamOn)
      (Development|SetHiddeninGame _beam_2 false false)
      (Rendering|Material|SetScalarParameterValueonMaterials _beam "FadeInTime" 0.1)
      (Rendering|Material|SetScalarParameterValueonMaterials _beam "FadeOutTIme" (+ (Variables|Default|GetTotalSeconds) (Variables|Default|GetPost%20Launch%20Duration)))
      (Rendering|Material|SetScalarParameterValue "/Game/Athena/Events/Generic/CountdownTimer/Materials/MPC_Countdown.MPC_Countdown" "DisableLightcone" 1.0 :!class /Script/Engine.KismetMaterialLibrary)
      (else
        (Development|SetHiddeninGame _beam_2 true true)
        (Rendering|Material|SetScalarParameterValueonMaterials _beam_2 "FadeInTime" (Variables|Default|GetPost%20Launch%20Duration))
        (Rendering|Material|SetScalarParameterValueonMaterials _beam "Fade In Time Length" 1.0)
        (Rendering|Material|SetScalarParameterValueonMaterials _beam "FadeOutTIme" (Variables|Default|GetParameterValue))
        (Rendering|Material|SetScalarParameterValue "/Game/Athena/Events/Generic/CountdownTimer/Materials/MPC_Countdown.MPC_Countdown" "DisableLightcone" 1.0 :!class /Script/Engine.KismetMaterialLibrary)))))

(fn OnRep_ParameterValue
  (bind _beam (Variables|Default|GetBeam))
  (Rendering|Material|SetScalarParameterValue "/Game/Athena/Events/Generic/CountdownTimer/Materials/MPC_Countdown.MPC_Countdown" "DisableLightcone" (Variables|Default|GetParameterValue) :!class /Script/Engine.KismetMaterialLibrary)
  (if (not (Variables|Default|GetBeamActive))
    (Rendering|Material|SetScalarParameterValueonMaterials _beam "FadeInTime" 0.1)
    (Rendering|Material|SetScalarParameterValueonMaterials _beam "Fade In Time Length" 0.0)
    (Rendering|Material|SetScalarParameterValueonMaterials _beam "FadeOutTIme" (- (Variables|Default|GetPost%20Launch%20Duration) (Variables|Default|GetTotalSeconds)))
    (Development|SetHiddeninGame _beam false false)))

(fn ConstructionScript)
